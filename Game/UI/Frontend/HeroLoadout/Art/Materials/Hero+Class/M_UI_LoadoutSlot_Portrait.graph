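# Translucent UI material for a hero portrait, named for the loadout slot in the frontend hero loadout screen.
# It samples the HeroPortrait texture parameter, defaulting to a Commando soldier image, and uses its alpha.
# A parabolic mask built from horizontal UV is boosted and clamped, then drives a lerp that fades that alpha toward the left and right edges.
# A slightly offset UV Rotator feeds nothing shown, and no material output connections appear.
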Material /Game/UI/Frontend/HeroLoadout/Art/Materials/Hero+Class/M_UI_LoadoutSlot_Portrait
graph {
 properties {
  bUsedWithStaticMesh=false
  blendMode=BLEND_Translucent
  materialDomain=MD_UI
  pixelDepthOffsetMode=PDOM_Legacy
  usedShadingModels=Default Lit
 }
 nodes [
  n1 = TextureCoordinate()
  n2 = ComponentMask(None=@n1) {r=true}
  n3 = Multiply(A=Multiply(A=@n2) {constB=2}, B=Multiply(A=OneMinus(None=@n2)) {constB=2})
  LinearInterpolate(
   B=TextureSampleParameter2D(UVs=TextureCoordinate()) {parameterName=HeroPortrait, texture=/Game/UI/Foundation/Textures/Icons/Heroes/Athena/Soldier/T-Soldier-HID-005-Athena-Commando-M}.A
   Alpha=Clamp(None=Multiply(A=@n3) {constB=1.5})
  )
  Rotator(Coordinate=@n1, Time=Constant() {r=-0.01}) {speed=6.2832}
 ]
}
